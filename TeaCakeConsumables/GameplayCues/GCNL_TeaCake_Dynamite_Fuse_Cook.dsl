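; Looping gameplay cue that animates the burning fuse while a player cooks the TeaCake dynamite bundle.
; Timelines, rate-scaled from fuse duration values, move sparks along a spline rebuilt from weapon-mesh sockets and drive wick fade and emissive flash.
; Stage one ramps spark opacity, color and size, then swaps to an attached death effect, fades audio and ramps again.
; Removal resets the wick material and deactivates the death effect.
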(event Custom|Control%20Stage%201%20Fuse%20Particles
  (bind _looping_particle (Variables|Default|GetLooping%20Particle))
  ;[1 1 1] Sets Values in Fuse Particle Systems
  (Utilities|IsValid _looping_particle
    (:Is%20Valid
      (Effects|Components|ParticleSystem|SetFloatParameter _looping_particle "AlphaIntensity" (Variables|Default|GetPSys%20Opacity))
      (Effects|Components|ParticleSystem|SetVectorParameter _looping_particle "ColorIntensity" (Variables|Default|GetPSys%20Color))
      (Effects|Components|ParticleSystem|SetVectorParameter _looping_particle "Size_Intensity" (Variables|Default|GetPsys%20Size))
      (Transformation|SetWorldLocation _looping_particle (Variables|Default|GetPSysSplineLoca) false false))))

(event Custom|Control%20Stage%202%20Fuse%20Particles
  (bind _deathfx (Variables|Default|GetDeathFX))
  ;[1 1 1] Sets Values in Fuse Particle Systems
  (Utilities|IsValid _deathfx
    (:Is%20Valid
      (Effects|Components|ParticleSystem|SetFloatParameter _deathfx "AlphaIntensity" (Variables|Default|GetPSys%20Opacity))
      (Effects|Components|ParticleSystem|SetVectorParameter _deathfx "ColorIntensity" (Variables|Default|GetPSys%20Color))
      (Effects|Components|ParticleSystem|SetVectorParameter _deathfx "Size_Intensity" (Variables|Default|GetPsys%20Size)))))

(event EventOnLoopingStart(EitherVFX-preferthisversion)
  (params (MyTarget :type Actor%20Object%20Reference) (Parameters :type Gameplay%20Cue%20Parameters%20Structure%20%28by%20ref%29) (ParticleComponents :type Array%20of%20FXSystem%20Component%20Object%20References) (AudioComponents :type Array%20of%20Audio%20Component%20Object%20References))
  (bind _dynamitebundle (Variables|Default|GetDynamiteBundle))
  (bind _returnvalue (Visual|GetWeaponMesh _dynamitebundle))
  (bind _fuseduration (Variables|Default|GetFuseDuration))
  (bind _asplayer_pawn_athena (Utilities|Casting|CastToPlayerPawn_Athena MyTarget))
  (bind _asb_athena_tea_cake_dynamite (Utilities|Casting|CastToB_Athena_TeaCake_Dynamite (Variables|/Script/FortniteGame.FortPawn|GetCurrentWeapon _asplayer_pawn_athena)))
  (Variables|Default|SetDynamiteBundle _asb_athena_tea_cake_dynamite)
  (Variables|Default|SetLooping%20Particle (Utilities|Array|Get(acopy) ParticleComponents))
  (Variables|Default|SetLooping%20Audio (Utilities|Array|Get(acopy) AudioComponents))
  (Variables|Default|SetFuseDuration (ScalableFloat|GetCachedScalableFloatValue (Variables|Default|GetHF_FuseDuration)))
  (bind _outxy (DataTable|EvaluateCurveTableRow "/Game/Athena/Balance/DataTables/AthenaGameData.AthenaGameData" "Default.TNT.RomanCandleDuration")
    (:RowFound
      (Variables|Default|SetFuse%202nd%20Stage%20Duration _outxy)
      (Components|Timeline|SetPlayRate (Variables|GCNL_TeaCake_Dynamite_Fuse_Cook|GetFuse%20Movement) (/ 1.0 _fuseduration))
      (Components|Timeline|SetPlayRate (Variables|GCNL_TeaCake_Dynamite_Fuse_Cook|GetFuse%20Stage%201%20Timeline) (/ 1.0 (- _fuseduration (Variables|Default|GetFuse%202nd%20Stage%20Duration))))
      (Components|Timeline|SetPlayRate (Variables|GCNL_TeaCake_Dynamite_Fuse_Cook|GetFuse%20Stage%202%20Timeline) (/ 1.0 (Variables|Default|GetFuse%202nd%20Stage%20Duration)))
      (Utilities|FlowControl|Sequence
        (:then_0
          ;[1 1 1] Movement of Fuse along Wick and Fade Material
          (bind (_direction _lerp_control _material_flash) (Timeline|Fuse%20Movement PlayFromStart)
            (:Update
              (bind _spline_location (CallFunction|SplineSetup _lerp_control))
              (Variables|Default|SetPSysSplineLoca _spline_location)
              (Rendering|Material|SetScalarParameterValueonMaterials _returnvalue "Fade along U" _lerp_control)
              (Rendering|Material|SetVectorParameterValueonMaterials _returnvalue "EmissiveMultiplier" (Math|Vector|Lerp(Vector) "0, 0, 0" (Variables|Default|GetEmissive%20Color) _material_flash)))))
        (:then_1
          (bind (_direction_2 _death_warning) (Timeline|Fuse%20Stage%201%20Timeline PlayFromStart)
            (:Update
              (bind (_opacity _color _size) (CallFunction|WarningSetup _death_warning))
              (Variables|Default|SetPSys%20Opacity _opacity)
              (Variables|Default|SetPSys%20Color _color)
              (Variables|Default|SetPsys%20Size _size)
              (CallFunction|ControlStage1FuseParticles))
            (:Finished
              (Components|Activation|Deactivate (Variables|Default|GetLooping%20Particle))
              (bind _returnvalue_2 (Effects|SpawnFXSystemAttached "/Game/Weapons/Grenades/TNT_Bundle/FX/NS_TNT_Fuse_Death_Converted.NS_TNT_Fuse_Death_Converted" (Visual|GetWeaponMesh (Variables|Default|GetDynamiteBundle)) "TNT_Bundle"))
              (bind _output_get (Variables|Default|SetDeathFX _returnvalue_2))
              (Effects|Components|ParticleSystem|SetFloatParameter _output_get "AlphaIntensity" (Variables|Default|GetPSys%20Opacity))
              (Effects|Components|ParticleSystem|SetVectorParameter _output_get "ColorIntensity" (Variables|Default|GetPSys%20Color))
              (Effects|Components|ParticleSystem|SetVectorParameter _output_get "Size_Intensity" (Variables|Default|GetPsys%20Size))
              (bind _looping_audio (Variables|Default|GetLooping%20Audio)
                (:then
                  (Audio|Components|Audio|FadeOut _looping_audio 0.25 0.0 "Linear" :!class /Script/Engine.AudioComponent))
                (:else
                  _))
              (bind (_direction_3 _death_warning_2) (Timeline|Fuse%20Stage%202%20Timeline PlayFromStart)
                (:Update
                  (bind (_opacity_2 _color_2 _size_2) (CallFunction|WarningSetup _death_warning_2))
                  (Variables|Default|SetPSys%20Opacity _opacity_2)
                  (Variables|Default|SetPSys%20Color _color_2)
                  (Variables|Default|SetPsys%20Size _size_2)
                  (CallFunction|ControlStage2FuseParticles))))))))))

(event EventOnRemoval(EitherVFX-preferthisversion)
  (params (MyTarget :type Actor%20Object%20Reference) (Parameters :type Gameplay%20Cue%20Parameters%20Structure%20%28by%20ref%29) (ParticleComponents :type Array%20of%20FXSystem%20Component%20Object%20References) (AudioComponents :type Array%20of%20Audio%20Component%20Object%20References) (BurstCameraShakeInstance :type Camera%20Shake%20Base%20Object%20Reference) (BurstDecalInstance :type Decal%20Actor%20Object%20Reference))
  (bind _deathfx (Variables|Default|GetDeathFX))
  ;[1 1 1] Ensures that wick is always at 100% if you have more than 1 bundle stack and Clean Up Spawned Particle System
  (bind _asplayer_pawn_athena (Utilities|Casting|CastToPlayerPawn_Athena MyTarget))
  (bind _asb_athena_tea_cake_dynamite (Utilities|Casting|CastToB_Athena_TeaCake_Dynamite (Variables|/Script/FortniteGame.FortPawn|GetCurrentWeapon _asplayer_pawn_athena)))
  (Rendering|Material|SetScalarParameterValueonMaterials (Visual|GetWeaponMesh _asb_athena_tea_cake_dynamite) "Fade along U" 0.0)
  (Rendering|Material|SetVectorParameterValueonMaterials (Visual|GetWeaponMesh _asb_athena_tea_cake_dynamite) "EmissiveMultiplier" "0, 0, 0")
  (Utilities|IsValid _deathfx
    (:Is%20Valid
      (Components|Activation|Deactivate _deathfx))
    (:Is%20Not%20Valid
      _))
  (Components|Timeline|Stop (Variables|GCNL_TeaCake_Dynamite_Fuse_Cook|GetFuse%20Stage%202%20Timeline)))

(fn OnCeaseRelevant
  (params (MyTarget :type Actor%20Object%20Reference) (Parameters :type Gameplay%20Cue%20Parameters%20Structure%20%28by%20ref%29))
  (Components|Timeline|Stop (Variables|GCNL_TeaCake_Dynamite_Fuse_Cook|GetFuse%20Movement))
  (bind _returnvalue (|CallParentFunction MyTarget Parameters))
  (return _returnvalue))

(fn SplineSetup
  (params (Alpha :type Float%20%28double-precision%29))
  (bind _dynamitebundle (Variables|Default|GetDynamiteBundle))
  (bind _spline (Variables|/TeaCakeConsumables/Gameplay/Dynamite/B_Athena_TeaCake_Dynamite.B_Athena_TeaCake_Dynamite_C|GetSpline _dynamitebundle))
  (Spline|SetSplinePoints _spline (Utilities|Array|MakeArray (Transformation|GetSocketLocation (Visual|GetWeaponMesh (Variables|Default|GetDynamiteBundle)) "FX_FuseBegin") (Transformation|GetSocketLocation (Visual|GetWeaponMesh (Variables|Default|GetDynamiteBundle)) "dyn_fuse2") (Transformation|GetSocketLocation (Visual|GetWeaponMesh (Variables|Default|GetDynamiteBundle)) "dyn_fuse1") (Transformation|GetSocketLocation (Visual|GetWeaponMesh (Variables|Default|GetDynamiteBundle)) "FX_FuseEnd")) "World" true)
  (return (Spline|GetLocationatDistanceAlongSpline _spline (Math|Float|Lerp 0.0 (Spline|GetSplineLength _spline) Alpha) "World")))

(fn ConstructionScript)

(fn WarningSetup
  (params (Warning :type Float%20%28double-precision%29))
  (return (Math|Float|Lerp 1.0 (LocalVariables|GetMax%20Possible%20Opacity) Warning) (Math|Vector|Lerp(Vector) "1.000000,1.000000,1.000000" (LocalVariables|GetFinal%20Color) Warning) (Math|Vector|Lerp(Vector) "1.000000,1.000000,1.000000" (LocalVariables|GetMax%20Possible%20Size) Warning)))
