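; Burst gameplay cue for the TeaCake dynamite bounce that aligns the scorch mark decal to the floor.
; OnBurst runs the parent effects, then sphere-traces a short distance down from the target actor.
; On a hit, it moves the spawned decal to the impact point and orients it to the surface normal, with random rotation to vary the mark.
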(fn OnBurst(EitherVFX-preferthisversion)
  (params (MyTarget :type Actor%20Object%20Reference) (Parameters :type Gameplay%20Cue%20Parameters%20Structure%20%28by%20ref%29) (ParticleComponents :type Array%20of%20FXSystem%20Component%20Object%20References) (AudioComponents :type Array%20of%20Audio%20Component%20Object%20References) (BurstCameraShakeInstance :type Camera%20Shake%20Base%20Object%20Reference) (BurstDecalInstance :type Decal%20Actor%20Object%20Reference))
  (|CallParentFunction MyTarget Parameters ParticleComponents AudioComponents BurstCameraShakeInstance BurstDecalInstance)
  ;[1 0.2696 6.1e-05 0.2435] Align scorch mark to floor
  (bind (_outhit _returnvalue) (Collision|SphereTraceByChannel MyTarget (Transformation|GetActorLocation MyTarget) (- (Transformation|GetActorLocation MyTarget) "0.000000,0.000000,10.000000") 1e+02 "TraceTypeQuery10"))
  (if _returnvalue
    (Transformation|SetWorldLocationAndRotation (Variables|/Script/Engine.DecalActor|GetDecal BurstDecalInstance) (.ImpactPoint (Collision|BreakHitResult _outhit)) (Math|Rotator|CombineRotators (Math|Rotator|MakeRotator 0.0 -9e+01 (Math|Random|RandomFloatinRange -9e+01 9e+01)) (Math|Rotator|MakeRotfromZX (.ImpactNormal (Collision|BreakHitResult _outhit)) (Math|Vector|MakeVector (Math|Random|RandomFloatinRange -1.0 1.0)))) false true))
  (return))
